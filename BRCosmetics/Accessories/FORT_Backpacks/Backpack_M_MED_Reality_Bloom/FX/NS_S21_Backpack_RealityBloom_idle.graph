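# Niagara system for the idle effect on the Season 21 Reality Bloom backpack cosmetic, a mostly blue ambient glow with drifting particles.
# Pulsing cone meshes, recurring glow sprites and a one-burst startup glow layer with a looping-lifetime particle that drives sprite and light renderers.
# CPU dust motes rise with curl noise and follow source movement, while GPU embers drift upward, their spawn alpha decreasing with emitter age.
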NiagaraSystem /BRCosmetics/Accessories/FORT_Backpacks/Backpack_M_MED_Reality_Bloom/FX/NS_S21_Backpack_RealityBloom_idle
dependencies {
 usedDataInterfaces [
  &NiagaraDataInterfaceCurve
  &NiagaraDataInterfaceColorCurve
  &NiagaraDataInterfaceVector2DCurve
  &NiagaraDataInterfaceVectorCurve
 ]
 usedDynamicInputs [
  &NormalizeFloat
  &VectorLength
  &SetBoolByFloatComparison
 ]
 usedModules [
  &self:SystemSpawnScript.SystemScriptSource.SystemScriptGraph.NiagaraNodeAssignment_0.SetVariables_#1
  &FN_Integrator
  &SystemState
 ]
}
emitters {
 Cone_ScreenGlow {
  emitterUpdateScript [
   &EmitterState()
   &SpawnBurst_Instantaneous(Spawn Count=1, Use Spawn Probability=false) {enabled=false}
   &SpawnRate(SpawnRate=5, Use Spawn Probability=true)
  ]
  particleSpawnScript [
   /Niagara/Modules/Spawn/Initialization/InitializeParticle(
    Color=<rgba 1 1 1 0.1>
    Lifetime=&RandomRangeFloat(Minimum=1.5, Maximum=2)
    Module.Write Color=true
    Module.Write Lifetime=true
    Module.Write Mass=false
    Module.Write Material Random=true
    Module.Write Position=false
    Module.Write Scale=false
    Module.Write SpriteRotation=true
    Module.Write SpriteSize=false
    Position=&SimulationPosition()
    Sprite Rotation=&RandomRangeFloat(Minimum=0, Maximum=360)
    Sprite Size=&Vector2DFromFloat(Value=1)
   )
   &SystemLocation(MaskSpawnGroup=false, Offset=<0 0 -3.5>)
   &DynamicMaterialParameters(
    Index 0 Param 1=0.5
    Module.Param0WriteEnabled=true
    Module.Param1WriteEnabled=true
    Module.Param2WriteEnabled=true
    Module.Param3WriteEnabled=true
   )
   &InitialMeshOrientation(
    Rotation=&RandomRangeVector(Minimum=<0.04 -0.25 0.04>, Maximum=<-0.04 -0.25 -0.04>)
    Rotation Coordinate Space=::NewEnumerator0
    Use Orientation Vector=true
    Use Rotation Vector=true
   )
  ]
  particleUpdateScript [
   &ParticleState()
   &SolveForcesAndVelocity(Clamp Velocity=false, Limit Acceleration=false)
   &ScaleColor(
    Scale Alpha=&FloatFromCurve(FloatCurve=DataInterface() {propertyValues {Curve {keys [<at 0 0 RCIM_Cubic>,<at 0.5 1.5 RCIM_Cubic>,<at 1 0 RCIM_Cubic>]}}})
    Scale RGB=&VectorFromFloat(Value=&FloatFromCurve(FloatCurve=DataInterface() {propertyValues {Curve {keys [<at 0.5 0.1 RCIM_Cubic>]}}}, Scale Curve=4))
    ScaleA=true
    ScaleRGB=true
   )
   /SetVariables(Particles.LightColor=&ScaleLinearColorByCurve(
     InitialValue=<rgba 0.07 0.3 1 1>
     ScaleCurve=DataInterface() {propertyValues {
       AlphaCurve {keys [{interpMode=RCIM_Cubic, tangentMode=RCTM_User, time=0, value=0}, {interpMode=RCIM_Cubic, tangentMode=RCTM_User, time=0.5, value=5}, <at 1 0>]}
       BlueCurve {keys [<at 0 1>]}
       GreenCurve {keys [<at 0 1>]}
       RedCurve {keys [<at 0 1>]}
      }}
     CurveIndex=$Particles.NormalizedAge
    )) {isSetParameters=true}
   &ScaleSpriteSize(Scale Factor=&Vector2DFromFloat(Value=&FloatFromCurve(FloatCurve=DataInterface() {propertyValues {Curve {keys [
          {interpMode=RCIM_Cubic, tangentMode=RCTM_User, time=0, value=0}
          {interpMode=RCIM_Cubic, tangentMode=RCTM_User, time=0.31, value=1}
          {interpMode=RCIM_Cubic, tangentMode=RCTM_User, time=0.7, value=1}
          {interpMode=RCIM_Cubic, tangentMode=RCTM_User, time=1, value=0}
         ]}}}))) {enabled=false}
   &CameraOffset(Camera Offset Amount=10) {enabled=false}
   &ScaleMeshSize(Scale Factor=&VectorFromCurve(
     VectorCurve=DataInterface() {propertyValues {
       XCurve {keys [<at 0 0.55>,<at 0.2 1>,<at 1 0.55>]}
       YCurve {keys [<at 0 0.55>,<at 0.2 1>,<at 1 0.55>]}
       ZCurve {keys [<at 0 0.55>,<at 0.2 1>,<at 1 0.55>]}
      }}
     Scale Curve=&RandomRangeVector(Minimum=<0.5 0.25 0.25>, Maximum=<1 0.25 0.25>)
    ))
   &CameraOffset(Camera Offset Amount=10) {enabled=false, module=CameraOffset001}
  ]
  rendererClasses [&NiagaraMeshRendererProperties]
  renderers [{rendererClass=&NiagaraMeshRendererProperties, rendererIndex=0}]
 }
 Dust_Motes {
  emitterUpdateScript [
   &EmitterState()
   &SpawnRate(Spawn Probability=0.3, SpawnRate=10, Use Spawn Probability=true)
  ]
  particleSpawnScript [
   /Niagara/Modules/Spawn/Initialization/V2/InitializeParticle(
    Color=<rgba 0.4654 0.7646 2 0.2>
    Lifetime=&RandomRangeFloat(Minimum=1.7, Maximum=2)
    Sprite Rotation Angle=&RandomRangeFloat(Minimum=0, Maximum=360)
    Sprite Rotation Mode=::NewEnumerator0
    Sprite Size Mode=::NewEnumerator3
    Uniform Sprite Size=&RandomRangeFloat(Minimum=10, Maximum=2.4)
    UsePositionOffset=false
    Write Material Random=true
   )
   &FN_SetSpawnGroup_v2(Number of Spawn Groups=6) {enabled=false}
   &SkeletalMeshLocation(
    Bone / Socket Index=$Particles.SpawnGroup
    Bone Sampling Mode=::NewEnumerator4
    Inherit Velocity=false
    MaskSpawnGroup=false
    Sampled Position Offset=&RandomRangeVector(Minimum=<-20 -20 -20>, Maximum=<20 20 20>)
    Skeletal Mesh=$User.SkeletalMesh
   ) {enabled=false, module=SkeletalMeshLocation001}
   &SphereLocation(
    MaskSpawnGroup=false
    Override Seed=false
    Sphere Radius=2
    Surface Only=false
    UseNonUniformScale=false
    UseOffset=true
    UseSphereOrientation=false
   )
  ]
  particleUpdateScript [
   &InheritSourceMovement(ApplyForce=false, ApplyPosition=true, ApplyPositionFalloff=false)
   &ParticleState()
   &GravityForce(Gravity=&RandomRangeVector(Minimum=<0 0 30>, Maximum=<1 1 20>))
   &CurlNoiseForce(Noise Strength=&RandomRangeFloat(Minimum=25, Maximum=10), Pan Noise=true, Pan Noise Field=<0 0 1>)
   &Drag(Drag=3, Use Linear Drag=true, Use Rotational Drag=false)
   &SolveForcesAndVelocity(Clamp Velocity=false, Limit Acceleration=false)
   &ScaleColor(
    Scale Alpha=&FloatFromCurve(FloatCurve=DataInterface() {propertyValues {Curve {keys [
         {interpMode=RCIM_Cubic, tangentMode=RCTM_User, time=0, value=0}
         {interpMode=RCIM_Cubic, tangentMode=RCTM_User, time=0.5, value=2}
         {tangentMode=RCTM_User, time=1, value=0}
        ]}}})
    Scale RGB=&VectorFromFloat(Value=&RandomRangeFloat(Minimum=0.75, Maximum=2))
    ScaleA=true
    ScaleRGB=true
   )
  ]
  rendererClasses [&NiagaraSpriteRendererProperties]
  renderers [
   {rendererClass=&NiagaraSpriteRendererProperties}
   {rendererClass=&NiagaraSpriteRendererProperties, rendererIndex=1}
  ]
 }
 Embers001 {
  simTarget=GPUComputeSim
  emitterUpdateScript [
   &EmitterState()
   &SpawnRate(Spawn Probability=0.4, SpawnRate=14, Use Spawn Probability=true)
  ]
  particleSpawnScript [
   &InitializeParticle(
    Lifetime Max=2.5
    Lifetime Min=1.5
    Lifetime Mode=::NewEnumerator1
    Sprite Size Mode=::NewEnumerator1
    Uniform Sprite Size Max=2.5
    Uniform Sprite Size Min=0.5
    UsePositionOffset=false
    Write Material Random=true
   ) {module=InitializeParticle001}
   /SetVariables(Particles.ScaleU:NiagaraFloat=&FloatFromCurve(
     FloatCurve=DataInterface() {propertyValues {Curve {keys [
         {arriveTangent=-1.6504, interpMode=RCIM_Cubic, leaveTangent=-1.6504, tangentMode=RCTM_User, time=-0.02, value=1.0266}
         <at 0.5 0.5>
         {arriveTangent=-2.0092, interpMode=RCIM_Cubic, leaveTangent=-2.0092, tangentMode=RCTM_User, time=2, value=0.35}
        ]}}}
     CurveIndex=$Emitter.Age
    )) {isSetParameters=true}
   &Color(Color=&MakeLinearColorFromVectorAndFloat("Vector (RGB)"=<0.6157 1 2>, "Float (Alpha)"=&Multiply_Float(A=1, B=$Particles.ScaleU)))
   &AddVelocity(Velocity=&RandomRangeVector(Minimum=<-10 -10 30>, Maximum=<10 10 15>, Evaluation Type=::NewEnumerator1))
   &SphereLocation(
    MaskSpawnGroup=false
    Override Seed=false
    Sphere Radius=2
    Surface Only=false
    UseNonUniformScale=false
    UseOffset=false
    UseSphereOrientation=false
   ) {module=SphereLocation001}
  ]
  particleUpdateScript [
   &ParticleState()
   &CurlNoiseForce(Noise Strength=150, Pan Noise=false) {enabled=false}
   &TorusLocation(
    Align Torus With Axis=true
    Handle Radius=0
    Large Radius:NiagaraFloat=&FloatFromCurve(
     FloatCurve=DataInterface() {propertyValues {Curve {keys [{arriveTangent=-2.8702, interpMode=RCIM_Cubic, leaveTangent=-2.8702, tangentMode=RCTM_User, time=0, value=1}, <at 1 0>]}}}
     Scale Curve=&FloatFromCurve(
      FloatCurve=DataInterface() {propertyValues {Curve {keys [<at -0.0318 1 RCIM_Cubic>, {arriveTangent=-0.688, interpMode=RCIM_Cubic, leaveTangent=-0.688, tangentMode=RCTM_User, time=2.0112, value=0.3}]}}}
      CurveIndex=$Emitter.Age
      Scale Curve=&RandomRangeFloat(Minimum=100, Maximum=300)
     )
    )
    MaskSpawnGroup=false
    Radius Position=0
    Torus Axis=&RandomVector(Vector Scale=1)
    Torus Distribution Mode:ENiagaraTorusDistributionMode=::NewEnumerator1
    Torus Origin:NiagaraPosition=$Engine.Owner.Position
    U Position:NiagaraFloat=&Add_Float(
     A=&RandomRangeFloat(Minimum=0, Maximum=1)
     B=&Multiply_Float(
      A=&FloatFromCurve(FloatCurve=DataInterface() {propertyValues {Curve {keys [<at 0 1>,<at 1 0>]}}}, Scale Curve=0.5)
      B=$Particles.Initial.ScaleU
     )
    )
    V Position=1.3106
   ) {enabled=false}
   &Drag(Drag=&RandomRangeFloat(Minimum=1.5, Maximum=2.2), Use Linear Drag=true, Use Rotational Drag=false)
   &SolveForcesAndVelocity(Clamp Velocity=false, Limit Acceleration=false)
   &ScaleColor(
    Scale Alpha=&FloatFromCurve(FloatCurve=DataInterface() {propertyValues {Curve {keys [
         <at 0 0 RCIM_Cubic>
         {arriveTangent=-1.007, interpMode=RCIM_Cubic, leaveTangent=-1.007, tangentMode=RCTM_User, time=0.1, value=0.95}
         <at 0.2169 0.0158>
        ]}}})
    Scale RGB=&VectorFromCurve(VectorCurve=DataInterface() {propertyValues {
       XCurve {keys [
         {interpMode=RCIM_Cubic, tangentMode=RCTM_User, time=0, value=0}
         {arriveTangent=1080.0369, interpMode=RCIM_Cubic, leaveTangent=1080.0369, tangentMode=RCTM_User, time=0.1, value=50}
         {arriveTangent=-71.4286, interpMode=RCIM_Cubic, leaveTangent=-71.4286, time=0.3, value=899.9999}
         <at 0.8 0>
        ]}
       YCurve {keys [
         {interpMode=RCIM_Cubic, tangentMode=RCTM_User, time=0, value=0}
         {arriveTangent=1080.0369, interpMode=RCIM_Cubic, leaveTangent=1080.0369, tangentMode=RCTM_User, time=0.1, value=50}
         {arriveTangent=-71.4286, interpMode=RCIM_Cubic, leaveTangent=-71.4286, time=0.3, value=899.9999}
         <at 0.8 0>
        ]}
       ZCurve {keys [
         {interpMode=RCIM_Cubic, tangentMode=RCTM_User, time=0, value=0}
         {arriveTangent=1080.0369, interpMode=RCIM_Cubic, leaveTangent=1080.0369, tangentMode=RCTM_User, time=0.1, value=50}
         {arriveTangent=-71.4286, interpMode=RCIM_Cubic, leaveTangent=-71.4286, time=0.3, value=899.9999}
         <at 0.8 0>
        ]}
      }})
    ScaleA=true
    ScaleRGB=true
   ) {enabled=false}
   &ScaleSpriteSize(Scale Factor=&Multiply_Vector2(
     A=&Vector2DFromCurve(
      Vector2Curve=DataInterface() {propertyValues {
        XCurve {keys [<at 0 0 RCIM_Cubic>, {arriveTangent=-0.2212, interpMode=RCIM_Cubic, leaveTangent=-0.2212, time=0.2, value=2}, <at 0.8728 -0.1931>]}
        YCurve {keys [
          <at 0.0009 0 RCIM_Cubic>
          {arriveTangent=8.7764, interpMode=RCIM_Cubic, leaveTangent=8.7764, time=0.1, value=2.75}
          {arriveTangent=-3.8082, interpMode=RCIM_Cubic, leaveTangent=-3.8082, time=0.3, value=2.625}
          <at 0.8728 -0.1931>
         ]}
       }}
      ScaleCurve=<1 1>
     )
     B=<1 1>
    ))
  ]
  rendererClasses [&NiagaraSpriteRendererProperties]
  renderers [{rendererClass=&NiagaraSpriteRendererProperties}]
 }
 Glow {
  emitterUpdateScript [
   &EmitterState()
   &SpawnRate(SpawnRate=0.4, Use Spawn Probability=false)
  ]
  particleSpawnScript [
   &InitializeParticle(Lifetime=6, Sprite Size Mode=::NewEnumerator3, Uniform Sprite Size=40, UsePositionOffset=true, Write Material Random=true) {module=InitializeParticle001}
   &Color(Color=<rgba 0.0631 0.3466 1 0.8>)
   &DynamicMaterialParameters(
    Index 0 Param 1=10
    Module.Param0WriteEnabled=true
    Module.Param1WriteEnabled=false
    Module.Param2WriteEnabled=false
    Module.Param3WriteEnabled=false
   ) {enabled=false}
   &BoxLocation(Box Offset=<0 -20 0>, Box Size=<0 0 0>, MaskSpawnGroup=false, Module.Override Seed=false, Rotate Box=false, Surface Only=false) {enabled=false, module=BoxLocation001}
  ]
  particleUpdateScript [
   &ParticleState()
   &ScaleColor(
    Scale Alpha=&FloatFromCurve(FloatCurve=DataInterface() {propertyValues {Curve {keys [<at 0 0 RCIM_Cubic>,<at 0.2 1 RCIM_Cubic>,<at 1 0>]}}})
    Scale RGB=<1.5 1.5 1.2>
    ScaleA=true
    ScaleRGB=true
   )
   &CameraOffset(Camera Offset Amount=5) {enabled=false}
   &ScaleSpriteSize()
  ]
  rendererClasses [&NiagaraSpriteRendererProperties]
  renderers [{rendererClass=&NiagaraSpriteRendererProperties}]
 }
 Glow001 {
  emitterUpdateScript [
   &EmitterState(Life Cycle Mode=::NewEnumerator1, Loop Behavior=::NewEnumerator1, Loop Duration=1, UseLoopDelay=true)
   &SpawnBurst_Instantaneous(Use Spawn Probability=false, UseLoopCountLimit=false)
  ]
  particleSpawnScript [
   &InitializeParticle(Lifetime=6, Sprite Size Mode=::NewEnumerator3, Uniform Sprite Size=40, UsePositionOffset=true, Write Material Random=true) {module=InitializeParticle001}
   &Color(Color=<rgba 0.0631 0.3466 1 0.4>)
   &DynamicMaterialParameters(
    Index 0 Param 1=10
    Module.Param0WriteEnabled=true
    Module.Param1WriteEnabled=false
    Module.Param2WriteEnabled=false
    Module.Param3WriteEnabled=false
   ) {enabled=false}
   &BoxLocation(Box Offset=<0 -20 0>, Box Size=<0 0 0>, MaskSpawnGroup=false, Module.Override Seed=false, Rotate Box=false, Surface Only=false) {enabled=false, module=BoxLocation001}
  ]
  particleUpdateScript [
   &ParticleState()
   &ScaleColor(
    Scale Alpha=&FloatFromCurve(FloatCurve=DataInterface() {propertyValues {Curve {keys [{arriveTangent=-0.1943, interpMode=RCIM_Cubic, leaveTangent=-0.1943, tangentMode=RCTM_User, time=0, value=1}, <at 1 0>]}}})
    Scale RGB=<1.5 1.5 1.2>
    ScaleA=true
    ScaleRGB=true
   ) {module=ScaleColor001}
   &ScaleSpriteSize(Uniform Curve Sprite Scale=DataInterface() {propertyValues {Curve {keys [{tangentMode=RCTM_User, time=0, value=0.5}, {tangentMode=RCTM_User, time=1, value=1}]}}}) {module=ScaleSpriteSize001}
  ]
  rendererClasses [&NiagaraSpriteRendererProperties]
  renderers [{rendererClass=&NiagaraSpriteRendererProperties}]
 }
 Light002 {
  emitterUpdateScript [
   &EmitterState(Life Cycle Mode=::NewEnumerator1, Loop Behavior=::NewEnumerator1, Loop Duration Mode=::NewEnumerator1, UseLoopDelay=false)
   &SpawnBurst_Instantaneous(Spawn Time=0.001, Use Spawn Probability=false, UseLoopCountLimit=false)
   &SpawnRate(SpawnRate=8, Use Spawn Probability=false) {enabled=false}
  ]
  particleSpawnScript [
   &InitializeParticle(Lifetime=3, Sprite Size Mode=::NewEnumerator3, Uniform Sprite Size=15, UsePositionOffset=true, Write Material Random=true) {module=InitializeParticle001}
   &Color(Color=&MakeLinearColorFromVectorAndFloat("Vector (RGB)"=<0.3 1 0>, "Float (Alpha)"=&ClampFloat(Float=1)))
   &BoxLocation(Box Offset=<10 0 0>, Box Size=<0 0 0>, MaskSpawnGroup=false, Module.Override Seed=false, Rotate Box=false, Surface Only=false) {enabled=false, module=BoxLocation001}
  ]
  particleUpdateScript [
   &ParticleState(
    Kill Particles When Lifetime Has Elapsed=false
    Let Infinitely Lived Particles Die When Emitter Deactivates=true
    Loop Particles Lifetime=true
   )
   &ScaleColor(
    Scale Alpha=&FloatFromCurve(FloatCurve=DataInterface() {propertyValues {Curve {keys [<at 0 1>,<at 1 0.15>]}}})
    ScaleA=true
    ScaleRGB=true
   )
   /SetVariables(Particles.Gen5LightColor:LinearColor=&ScaleLinearColorByCurve(
     InitialValue=<rgba 0.0856 0.272 1 1>
     ScaleCurve=DataInterface() {propertyValues {
       AlphaCurve {keys [{interpMode=RCIM_Cubic, tangentMode=RCTM_User, time=0, value=4}]}
       BlueCurve {keys [<at 0 1>,<at 1 1>]}
       GreenCurve {keys [<at 0 1>,<at 1 1>]}
       RedCurve {keys [<at 0 1>,<at 1 1>]}
      }}
     CurveIndex=$Particles.NormalizedAge
    )) {isSetParameters=true}
   /SetVariables(Particles.Gen5LightIntensity:NiagaraFloat=&FloatFromCurve(FloatCurve=DataInterface() {propertyValues {Curve {keys [<at 0 150 RCIM_Cubic>]}}})) {enabled=false, isSetParameters=true}
   &CameraOffset(Camera Offset Amount=30) {enabled=false}
   /SetVariables(Particles.LightColor=&ScaleLinearColorByCurve(
     InitialValue=<rgba 1 0.2753 0.0949 1>
     ScaleCurve=DataInterface() {propertyValues {
       AlphaCurve {keys [{interpMode=RCIM_Cubic, tangentMode=RCTM_User, time=0, value=1}, {interpMode=RCIM_Cubic, tangentMode=RCTM_User, time=0.5, value=5}, <at 1 1>]}
       BlueCurve {keys [<at 0 1>]}
       GreenCurve {keys [<at 0 1>]}
       RedCurve {keys [<at 0 1>]}
      }}
     CurveIndex=$Particles.NormalizedAge
    )) {enabled=false, isSetParameters=true}
  ]
  rendererClasses [&NiagaraSpriteRendererProperties, &NiagaraLightRendererProperties]
  renderers [
   {rendererClass=&NiagaraSpriteRendererProperties}
   {rendererClass=&NiagaraLightRendererProperties, rendererIndex=1}
  ]
 }
}
summary {
 systemName=NS_S21_Backpack_RealityBloom_idle
 emitters [
  {emitterName=Dust_Motes, rendererClasses [&NiagaraSpriteRendererProperties]}
  {emitterName=Glow, rendererClasses [&NiagaraSpriteRendererProperties]}
  {emitterName=Light002, rendererClasses [&NiagaraSpriteRendererProperties, &NiagaraLightRendererProperties]}
  {emitterName=Glow001, rendererClasses [&NiagaraSpriteRendererProperties]}
  {emitterName=Embers001, simTarget=GPUComputeSim, rendererClasses [&NiagaraSpriteRendererProperties]}
  {emitterName=Cone_ScreenGlow, rendererClasses [&NiagaraMeshRendererProperties]}
 ]
 userVariables [
  User.Color_Tint {type=LinearColor, description="", defaultValue {struct=LinearColor, value=<rgba 1 1 1 1>}}
  User.EmissiveIntensity {type=NiagaraFloat, defaultValue=0, description=""}
  User.IsSkydiving {type=NiagaraBool}
  User.Landing {type=NiagaraBool}
  User.RibbonIDOffset {type=NiagaraInt32}
  User.SkeletalMesh {
   type=NiagaraDataInterfaceSkeletalMesh
   defaultValue {
    struct=NiagaraExt_VariableValue_DataInterface
    value {dataInterface=&self:NiagaraDataInterfaceSkeletalMesh_0}
   }
  }
 ]
}
