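# Physics asset for the Reality Bloom back bling mesh, giving its pods and stem bodies for secondary motion.
# Capsule bodies form chains from base bones for left, right and back pods and a stem, which forks three ways at its second dynamic bone.
# Constraints link each chain with swing limits of a few degrees and zero twist limit, keeping sway subtle.
# A large unconstrained capsule covers the pack.
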PhysicsAsset /BRCosmetics/Accessories/FORT_Backpacks/Backpack_M_MED_Reality_Bloom/Meshes/M_MED_Reality_Bloom_Pack_Physics
bodies {
 Back_pod_base [{shapeType=Capsule, length=0.2, radius=0.505, shapeName=Back_pod_base_capsule, rotation=<rot 0 -90 0>}]
 L_pod_base [{shapeType=Capsule, length=0.505, radius=0.505, shapeName=L_pod_base_capsule, rotation=<rot 0 -90 0>}]
 R_pod_base [{shapeType=Capsule, length=0.2, radius=0.505, shapeName=R_pod_base_capsule, rotation=<rot 0 -90 0>}]
 Stem_base [{shapeType=Capsule, length=0.505, radius=0.505, shapeName=Stem_base_capsule, rotation=<rot 0 -90 0>}]
 dyn_Back_pod_1 [{shapeType=Capsule, length=2.9801, radius=1.163, shapeName=dyn_Back_pod_1_capsule, center=<3 0 0>, rotation=<rot 90 0 0>}]
 dyn_Back_pod_2 [{shapeType=Capsule, length=2.3731, radius=0.9324, shapeName=dyn_Back_pod_2_capsule, center=<2 0 0>, rotation=<rot 90 0 0>}]
 dyn_Back_pod_3 [{shapeType=Capsule, length=2.676, radius=0.7628, shapeName=dyn_Back_pod_3_capsule, center=<2.0789 0 0>, rotation=<rot 90 0 0>}]
 dyn_Back_pod_4 [{shapeType=Capsule, length=0.505, radius=0.505, shapeName=dyn_Back_pod_4_capsule, center=<0.8 0 0>, rotation=<rot 90 0 0>}]
 dyn_L_pod_1 [{shapeType=Capsule, length=2.0115, radius=1.2339, shapeName=dyn_L_pod_1_capsule, center=<2.4717 0 0>, rotation=<rot 90 0 0>}]
 dyn_L_pod_2 [{shapeType=Capsule, length=1.3897, radius=0.6881, shapeName=dyn_L_pod_2_capsule, center=<2 0 0>, rotation=<rot 90 0 0>}]
 dyn_L_pod_3 [{shapeType=Capsule, length=0.5477, radius=0.5334, shapeName=dyn_L_pod_3_capsule, center=<0.2878 0 0>, rotation=<rot 90 0 0>}]
 dyn_R_pod_1 [{shapeType=Capsule, length=2.3085, radius=1.3928, shapeName=dyn_R_pod_1_capsule, center=<2 0 0>, rotation=<rot 90 0 0>}]
 dyn_R_pod_2 [{shapeType=Capsule, length=1.0637, radius=0.6937, shapeName=dyn_R_pod_2_capsule, center=<1.5 0 0>, rotation=<rot 90 0 0>}]
 dyn_R_pod_3 [{shapeType=Capsule, length=0.7286, radius=0.6591, shapeName=dyn_R_pod_3_capsule, center=<1 0 0>, rotation=<rot 90 0 0>}]
 dyn_Stem_1 [{shapeType=Capsule, length=3.0964, radius=1.0355, shapeName=dyn_Stem_1_capsule, center=<2.5 0 0>, rotation=<rot 90 0 0>}]
 dyn_Stem_2 [{shapeType=Capsule, length=4.6096, radius=1.1307, shapeName=dyn_Stem_2_capsule, center=<4.1613 0 0>, rotation=<rot 90 0 0>}]
 dyn_Stem_3 [{shapeType=Capsule, length=3.5202, radius=1.2673, shapeName=dyn_Stem_3_capsule, center=<3 0 0>, rotation=<rot 90 0 0>}]
 dyn_Stem_4 [{shapeType=Capsule, length=1.7037, radius=1.7552, shapeName=dyn_Stem_4_capsule, center=<2.7477 0 0>, rotation=<rot 90 0 0>}]
 dyn_Stem_5 [{shapeType=Capsule, length=0.5923, radius=1.4855, shapeName=dyn_Stem_5_capsule, center=<1.8133 0 0>, rotation=<rot 90 0 0>}]
 dyn_Stem_6 [{shapeType=Capsule, length=2.566, radius=1.2105, shapeName=dyn_Stem_6_capsule, center=<1.1256 0 0>, rotation=<rot 90 0 0>}]
 dyn_Stem_7 [{shapeType=Capsule, length=1.0019, radius=1.3247, shapeName=dyn_Stem_7_capsule, center=<1.5268 0 0>, rotation=<rot 90 0 0>}]
 pack [{shapeType=Capsule, length=8.3749, radius=9.8208, shapeName=pack_capsule, center=<0.289 -10.7859 -6.5096>}]
}
constraints [
 {bone1Name=dyn_R_pod_1, bone2Name=R_pod_base, swing1LimitDegrees=5, swing2LimitDegrees=2.5, twistLimitDegrees=0}
 {bone1Name=dyn_R_pod_2, bone2Name=dyn_R_pod_1, swing1LimitDegrees=5, swing2LimitDegrees=2.5, twistLimitDegrees=0}
 {bone1Name=dyn_R_pod_3, bone2Name=dyn_R_pod_2, swing1LimitDegrees=5, swing2LimitDegrees=2.5, twistLimitDegrees=0}
 {bone1Name=dyn_Back_pod_1, bone2Name=Back_pod_base, swing1LimitDegrees=8, swing2LimitDegrees=2.5, twistLimitDegrees=0}
 {bone1Name=dyn_Back_pod_2, bone2Name=dyn_Back_pod_1, swing1LimitDegrees=6, swing2LimitDegrees=2.5, twistLimitDegrees=0}
 {bone1Name=dyn_Back_pod_3, bone2Name=dyn_Back_pod_2, swing1LimitDegrees=6, swing2LimitDegrees=2.5, twistLimitDegrees=0}
 {bone1Name=dyn_Back_pod_4, bone2Name=dyn_Back_pod_3, swing1LimitDegrees=6, swing2LimitDegrees=2.5, twistLimitDegrees=0}
 {bone1Name=dyn_Stem_1, bone2Name=Stem_base, swing1LimitDegrees=2, swing2LimitDegrees=2, twistLimitDegrees=0}
 {bone1Name=dyn_Stem_2, bone2Name=dyn_Stem_1, swing1LimitDegrees=5, swing2LimitDegrees=5, twistLimitDegrees=0}
 {bone1Name=dyn_Stem_3, bone2Name=dyn_Stem_2, swing1LimitDegrees=5, swing2LimitDegrees=5, twistLimitDegrees=0}
 {bone1Name=dyn_Stem_4, bone2Name=dyn_Stem_2, swing1LimitDegrees=5, swing2LimitDegrees=5, twistLimitDegrees=0}
 {bone1Name=dyn_Stem_5, bone2Name=dyn_Stem_4, swing1LimitDegrees=2.5, swing2LimitDegrees=2.5, twistLimitDegrees=0}
 {bone1Name=dyn_Stem_6, bone2Name=dyn_Stem_2, swing1LimitDegrees=3, swing2LimitDegrees=3, twistLimitDegrees=0}
 {bone1Name=dyn_Stem_7, bone2Name=dyn_Stem_6, swing1LimitDegrees=2.5, swing2LimitDegrees=2.5, twistLimitDegrees=0}
 {bone1Name=dyn_L_pod_1, bone2Name=L_pod_base, swing1LimitDegrees=5, swing2LimitDegrees=2.5, twistLimitDegrees=0}
 {bone1Name=dyn_L_pod_2, bone2Name=dyn_L_pod_1, swing1LimitDegrees=5, swing2LimitDegrees=2.5, twistLimitDegrees=0}
 {bone1Name=dyn_L_pod_3, bone2Name=dyn_L_pod_2, swing1LimitDegrees=5, swing2LimitDegrees=2.5, twistLimitDegrees=0}
]
